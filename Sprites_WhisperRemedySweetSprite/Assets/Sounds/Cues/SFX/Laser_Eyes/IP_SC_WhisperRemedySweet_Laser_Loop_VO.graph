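# Sound cue for the Whisper Remedy Sweet sprite's laser eyes loop voice, drawing on long shoot VO recordings.
# A Random node picks one of four equally weighted waves, a single-input Mixer boosts it, and a root Modulator adds variation with a raised pitch floor.
# Despite the Loop name, no Looping node appears in the cue.
# Playback uses a close VO attenuation asset and a third-person ranged weapon fire sound class.
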SoundCue /Sprites_WhisperRemedySweetSprite/Assets/Sounds/Cues/SFX/Laser_Eyes/IP_SC_WhisperRemedySweet_Laser_Loop_VO
root=Modulator_3
nodes [
 {id=Mixer_2, children [Random_0], properties {inputVolume [1.3]}}
 {id=Modulator_3, children [Mixer_2], properties {pitchMin=1}}
 {id=Random_0, children [WavePlayer_0,WavePlayer_1,WavePlayer_2,WavePlayer_3], properties {weights [1,1,1,1]}}
 {
  id=WavePlayer_0
  properties {soundWaveAssetPtr=/Sprites_WhisperRemedySweetSprite/Assets/Sounds/Waves/VO/Laser_Eyes/IP_WhisperRemedySweet_Shoot_VO_Long_01}
 }
 {
  id=WavePlayer_1
  properties {soundWaveAssetPtr=/Sprites_WhisperRemedySweetSprite/Assets/Sounds/Waves/VO/Laser_Eyes/IP_WhisperRemedySweet_Shoot_VO_Long_02}
 }
 {
  id=WavePlayer_2
  properties {soundWaveAssetPtr=/Sprites_WhisperRemedySweetSprite/Assets/Sounds/Waves/VO/Laser_Eyes/IP_WhisperRemedySweet_Shoot_VO_Long_03}
 }
 {
  id=WavePlayer_3
  properties {soundWaveAssetPtr=/Sprites_WhisperRemedySweetSprite/Assets/Sounds/Waves/VO/Laser_Eyes/IP_WhisperRemedySweet_Shoot_VO_Long_04}
 }
]
properties {
 attenuationSettings=/Sprites_WhisperRemedySweetSprite/Assets/Sounds/Settings/Laser_Eyes/IP_WhisperRemedySweet_Sprite_VO_Close_Att
 maxDistance=4000
 soundClassObject=&Fort_WeaponRanged_Fire_3P_PSM
 volumeMultiplier=0.8
}
