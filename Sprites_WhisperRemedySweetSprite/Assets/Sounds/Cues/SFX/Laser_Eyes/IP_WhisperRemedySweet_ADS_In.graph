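# Sound cue for the Whisper Remedy Sweet sprite's laser-eyes ADS-in sound, played as player foley.
# A mixer layers a fixed foley wave, a random pick of four step chimes, and a random pick of four water-grab sounds that starts after a short fixed delay.
# Each layer passes through a modulator, and the cue uses sprite foley attenuation with a concurrency override that blocks new instances beyond two.
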SoundCue /Sprites_WhisperRemedySweetSprite/Assets/Sounds/Cues/SFX/Laser_Eyes/IP_WhisperRemedySweet_ADS_In
root=Mixer_1
nodes [
 {id=Delay_0, children [Modulator_4], properties {delayMax=0.15, delayMin=0.15}}
 {id=Mixer_1, children [Modulator_3,Modulator_2,Delay_0], properties {inputVolume [1,1,1]}}
 {id=Modulator_2, children [Random_0], properties {volumeMin=0.8}}
 {id=Modulator_3, children [WavePlayer_0], properties {volumeMin=0.8}}
 {id=Modulator_4, children [Random_1], properties {volumeMin=0.8}}
 {id=Random_0, children [WavePlayer_1,WavePlayer_2,WavePlayer_3,WavePlayer_4], properties {weights [1,1,1,1]}}
 {id=Random_1, children [WavePlayer_5,WavePlayer_6,WavePlayer_7,WavePlayer_8], properties {weights [1,1,1,1]}}
 {
  id=WavePlayer_0
  properties {soundWaveAssetPtr=/Sprites_WhisperRemedySweetSprite/Assets/Sounds/Waves/SFX/Laser_Eyes/IP_WhisperRemedySweet_Foley}
 }
 {id=WavePlayer_1, properties {soundWaveAssetPtr=/Sprites_WhisperRemedySweetSprite/Assets/Sounds/Waves/SFX/Sprites_Step_Chime_01}}
 {id=WavePlayer_2, properties {soundWaveAssetPtr=/Sprites_WhisperRemedySweetSprite/Assets/Sounds/Waves/SFX/Sprites_Step_Chime_02}}
 {id=WavePlayer_3, properties {soundWaveAssetPtr=/Sprites_WhisperRemedySweetSprite/Assets/Sounds/Waves/SFX/Sprites_Step_Chime_03}}
 {id=WavePlayer_4, properties {soundWaveAssetPtr=/Sprites_WhisperRemedySweetSprite/Assets/Sounds/Waves/SFX/Sprites_Step_Chime_04}}
 {id=WavePlayer_5, properties {soundWaveAssetPtr=/Sprites_WhisperRemedySweetSprite/Assets/Sounds/Waves/SFX/Sprite_Water_Grab_05}}
 {id=WavePlayer_6, properties {soundWaveAssetPtr=/Sprites_WhisperRemedySweetSprite/Assets/Sounds/Waves/SFX/Sprite_Water_Grab_06}}
 {id=WavePlayer_7, properties {soundWaveAssetPtr=/Sprites_WhisperRemedySweetSprite/Assets/Sounds/Waves/SFX/Sprite_Water_Grab_07}}
 {id=WavePlayer_8, properties {soundWaveAssetPtr=/Sprites_WhisperRemedySweetSprite/Assets/Sounds/Waves/SFX/Sprite_Water_Grab_08}}
]
properties {
 attenuationSettings=/Game/Sounds/Fort_Foley_Sounds/FoleyLibraries/Settings/Foley_Sprites_Attenuation_Basic_Att
 bOverrideConcurrency=true
 maxDistance=1536
 soundClassObject=&Fort_Player_Foley
 volumeMultiplier=0.5
 concurrencyOverrides {maxCount=2, resolutionRule=PreventNew}
}
