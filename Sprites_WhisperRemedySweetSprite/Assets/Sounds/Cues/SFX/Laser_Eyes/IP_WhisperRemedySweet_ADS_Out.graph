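# Sound cue for the Whisper Remedy Sweet sprite's Laser Eyes foley, apparently played when leaving aim-down-sights.
# A single foley wave runs through two chained Modulators for random variation, one with a widened lower volume bound.
# The cue lowers pitch and volume, uses the player foley sound class and shared sprite attenuation.
# Its concurrency override caps simultaneous plays at two and rejects new ones beyond that.
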SoundCue /Sprites_WhisperRemedySweetSprite/Assets/Sounds/Cues/SFX/Laser_Eyes/IP_WhisperRemedySweet_ADS_Out
root=Modulator_1
nodes [
 {id=Modulator_1, children [Modulator_3]}
 {id=Modulator_3, children [WavePlayer_0], properties {volumeMin=0.8}}
 {
  id=WavePlayer_0
  properties {soundWaveAssetPtr=/Sprites_WhisperRemedySweetSprite/Assets/Sounds/Waves/SFX/Laser_Eyes/IP_WhisperRemedySweet_Foley}
 }
]
properties {
 attenuationSettings=/Game/Sounds/Fort_Foley_Sounds/FoleyLibraries/Settings/Foley_Sprites_Attenuation_Basic_Att
 bOverrideConcurrency=true
 maxDistance=1536
 pitchMultiplier=0.6
 soundClassObject=&Fort_Player_Foley
 volumeMultiplier=0.3
 concurrencyOverrides {maxCount=2, resolutionRule=PreventNew}
}
