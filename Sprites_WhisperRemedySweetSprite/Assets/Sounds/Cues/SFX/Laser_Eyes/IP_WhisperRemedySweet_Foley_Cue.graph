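# Sound cue for the WhisperRemedySweet sprite's Laser Eyes foley, layering an effect with a voice line.
# A mixer combines the foley wave, through a modulator with slight volume variation, with a randomly chosen one of four inventory "Impulsed" VO waves, the voice mixed louder.
# It plays in the player foley sound class with sprite foley attenuation, and its concurrency override caps it at two instances, preventing new ones beyond that.
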SoundCue /Sprites_WhisperRemedySweetSprite/Assets/Sounds/Cues/SFX/Laser_Eyes/IP_WhisperRemedySweet_Foley_Cue
root=Mixer_0
nodes [
 {id=Mixer_0, children [Modulator_3,Random_2], properties {inputVolume [1,2]}}
 {id=Modulator_3, children [WavePlayer_0], properties {volumeMin=0.8}}
 {
  id=Random_2
  children [WavePlayer_13, WavePlayer_14, WavePlayer_15, WavePlayer_17]
  properties {bShouldExcludeFromBranchCulling=true, weights [1,1,1,1]}
 }
 {
  id=WavePlayer_0
  properties {soundWaveAssetPtr=/Sprites_WhisperRemedySweetSprite/Assets/Sounds/Waves/SFX/Laser_Eyes/IP_WhisperRemedySweet_Foley}
 }
 {
  id=WavePlayer_13
  properties {soundWaveAssetPtr=/Sprites_WhisperRemedySweetSprite/Assets/Sounds/Waves/VO/VO_S33_WhisperRemedySweetSprite_Inventory_Impulsed_011}
 }
 {
  id=WavePlayer_14
  properties {soundWaveAssetPtr=/Sprites_WhisperRemedySweetSprite/Assets/Sounds/Waves/VO/VO_S33_WhisperRemedySweetSprite_Inventory_Impulsed_014}
 }
 {
  id=WavePlayer_15
  properties {soundWaveAssetPtr=/Sprites_WhisperRemedySweetSprite/Assets/Sounds/Waves/VO/VO_S33_WhisperRemedySweetSprite_Inventory_Impulsed_015}
 }
 {
  id=WavePlayer_17
  properties {soundWaveAssetPtr=/Sprites_WhisperRemedySweetSprite/Assets/Sounds/Waves/VO/VO_S33_WhisperRemedySweetSprite_Inventory_Impulsed_017}
 }
]
properties {
 attenuationSettings=/Game/Sounds/Fort_Foley_Sounds/FoleyLibraries/Settings/Foley_Sprites_Attenuation_Basic_Att
 bOverrideConcurrency=true
 maxDistance=1536
 soundClassObject=&Fort_Player_Foley
 volumeMultiplier=1
 concurrencyOverrides {maxCount=2, resolutionRule=PreventNew}
}
