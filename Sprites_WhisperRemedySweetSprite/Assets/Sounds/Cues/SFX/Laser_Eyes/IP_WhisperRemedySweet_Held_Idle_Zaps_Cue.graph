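# Sound cue for the laser-eye zap SFX of the WhisperRemedySweet sprite, apparently in its held idle state.
# A quality switch picks randomly among six zap waves for the first two levels and a three-wave subset for the last two.
# A root modulator adds random variation to the output.
# Playback is quiet, spatially attenuated, routed through a third-person ranged-weapon-fire sound class, and capped at two voices with new plays rejected.
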SoundCue /Sprites_WhisperRemedySweetSprite/Assets/Sounds/Cues/SFX/Laser_Eyes/IP_WhisperRemedySweet_Held_Idle_Zaps_Cue
root=Modulator_3
nodes [
 {id=Modulator_3, children [QualityLevel_0], properties {volumeMin=0.8}}
 {id=QualityLevel_0, children [Random_1,Random_1,Random_2,Random_2]}
 {
  id=Random_1
  children [WavePlayer_5, WavePlayer_6, WavePlayer_7, WavePlayer_8, WavePlayer_9, WavePlayer_10]
  properties {weights [1,1,1,1,1,1]}
 }
 {id=Random_2, children [WavePlayer_5, WavePlayer_8, WavePlayer_10], properties {bShouldExcludeFromBranchCulling=true, weights [1,1,1]}}
 {
  id=WavePlayer_10
  properties {soundWaveAssetPtr=/Sprites_WhisperRemedySweetSprite/Assets/Sounds/Waves/SFX/Laser_Eyes/IP_WhisperRemedySweet_Zaps_06}
 }
 {
  id=WavePlayer_5
  properties {soundWaveAssetPtr=/Sprites_WhisperRemedySweetSprite/Assets/Sounds/Waves/SFX/Laser_Eyes/IP_WhisperRemedySweet_Zaps_01}
 }
 {
  id=WavePlayer_6
  properties {soundWaveAssetPtr=/Sprites_WhisperRemedySweetSprite/Assets/Sounds/Waves/SFX/Laser_Eyes/IP_WhisperRemedySweet_Zaps_02}
 }
 {
  id=WavePlayer_7
  properties {soundWaveAssetPtr=/Sprites_WhisperRemedySweetSprite/Assets/Sounds/Waves/SFX/Laser_Eyes/IP_WhisperRemedySweet_Zaps_03}
 }
 {
  id=WavePlayer_8
  properties {soundWaveAssetPtr=/Sprites_WhisperRemedySweetSprite/Assets/Sounds/Waves/SFX/Laser_Eyes/IP_WhisperRemedySweet_Zaps_04}
 }
 {
  id=WavePlayer_9
  properties {soundWaveAssetPtr=/Sprites_WhisperRemedySweetSprite/Assets/Sounds/Waves/SFX/Laser_Eyes/IP_WhisperRemedySweet_Zaps_05}
 }
]
properties {
 attenuationSettings=/Sprites_WhisperRemedySweetSprite/Assets/Sounds/Settings/Laser_Eyes/IP_WhisperRemedySweet_Zaps_Att
 bOverrideConcurrency=true
 maxDistance=2250
 soundClassObject=&Fort_WeaponRanged_Fire_3P_PSM
 volumeMultiplier=0.05
 concurrencyOverrides {maxCount=2, resolutionRule=PreventNew}
}
