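# Looping laser-eyes beam sound cue for the Whisper Remedy Sweet sprite, routed to the third-person ranged weapon fire sound class.
# Two looping waves, the main loop and a WB layer, each pass through an identical short fixed delay and the shared Beam Fire attenuation.
# A mixer blends them with the main loop at half the layer's level, then a root modulator adds slight random variation.
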SoundCue /Sprites_WhisperRemedySweetSprite/Assets/Sounds/Cues/SFX/Laser_Eyes/IP_WhisperRemedySweet_Laser_Loop_Cue
root=Modulator_3
nodes [
 {
  id=Attenuation_0
  children [Delay_0]
  properties {attenuationSettings=/Sprites_WhisperRemedySweetSprite/Assets/Sounds/Settings/Laser_Eyes/IP_WhisperRemedySweet_Beam_Fire_Att}
 }
 {
  id=Attenuation_1
  children [Delay_1]
  properties {attenuationSettings=&IP_WhisperRemedySweet_Beam_Fire_Att}
 }
 {id=Delay_0, children [WavePlayer_5], properties {delayMax=0.05, delayMin=0.05}}
 {id=Delay_1, children [WavePlayer_0], properties {delayMax=0.05, delayMin=0.05}}
 {id=Mixer_2, children [Attenuation_0, Attenuation_1], properties {inputVolume [0.5,1]}}
 {id=Modulator_3, children [Mixer_2], properties {pitchMin=1}}
 {
  id=WavePlayer_0
  properties {
   bLooping=true
   soundWaveAssetPtr=/Sprites_WhisperRemedySweetSprite/Assets/Sounds/Waves/SFX/Laser_Eyes/IP_WhisperRemedySweet_Laser_Loop_WB_Layer
  }
 }
 {
  id=WavePlayer_5
  properties {bLooping=true, soundWaveAssetPtr=/Sprites_WhisperRemedySweetSprite/Assets/Sounds/Waves/SFX/Laser_Eyes/IP_WhisperRemedySweet_Laser_Loop}
 }
]
properties {maxDistance=18400, soundClassObject=&Fort_WeaponRanged_Fire_3P_PSM, volumeMultiplier=0.8}
