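# Sound cue for the WhisperRemedySweet sprite's laser-eyes attack stopping, pairing a laser shut-off effect with a relieved voice reaction.
# A Mixer layers a random pick of three shoot-stop waves, volume-varied by a Modulator, with a random pick of four relieved VO lines.
# The VO layer sits at half volume behind a fixed quarter-second delay and is excluded from branch culling.
# It uses beam-fire attenuation and a third-person ranged-weapon fire sound class.
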SoundCue /Sprites_WhisperRemedySweetSprite/Assets/Sounds/Cues/SFX/Laser_Eyes/IP_WhisperRemedySweet_Laser_Shoot_Stop_Cue
root=Mixer_2
nodes [
 {id=Delay_0, children [Random_2], properties {delayMax=0.25, delayMin=0.25}}
 {id=Mixer_2, children [Modulator_4,Delay_0], properties {inputVolume [1,0.5]}}
 {id=Modulator_4, children [Random_1], properties {volumeMin=0.8}}
 {id=Random_1, children [WavePlayer_6,WavePlayer_7,WavePlayer_8], properties {weights [1,1,1]}}
 {
  id=Random_2
  children [WavePlayer_9, WavePlayer_10, WavePlayer_11, WavePlayer_12]
  properties {bShouldExcludeFromBranchCulling=true, weights [1,1,1,1]}
 }
 {
  id=WavePlayer_10
  properties {soundWaveAssetPtr=/Sprites_WhisperRemedySweetSprite/Assets/Sounds/Waves/VO/VO_S33_WhisperRemedySweetSprite_Emotion_Relieved_016}
 }
 {
  id=WavePlayer_11
  properties {soundWaveAssetPtr=/Sprites_WhisperRemedySweetSprite/Assets/Sounds/Waves/VO/VO_S33_WhisperRemedySweetSprite_Emotion_Relieved_014}
 }
 {
  id=WavePlayer_12
  properties {soundWaveAssetPtr=/Sprites_WhisperRemedySweetSprite/Assets/Sounds/Waves/VO/VO_S33_WhisperRemedySweetSprite_Emotion_Relieved_010}
 }
 {
  id=WavePlayer_6
  properties {soundWaveAssetPtr=/Sprites_WhisperRemedySweetSprite/Assets/Sounds/Waves/SFX/Laser_Eyes/IP_WhisperRemedySweet_Laser_ShootStop_01}
 }
 {
  id=WavePlayer_7
  properties {soundWaveAssetPtr=/Sprites_WhisperRemedySweetSprite/Assets/Sounds/Waves/SFX/Laser_Eyes/IP_WhisperRemedySweet_Laser_ShootStop_02}
 }
 {
  id=WavePlayer_8
  properties {soundWaveAssetPtr=/Sprites_WhisperRemedySweetSprite/Assets/Sounds/Waves/SFX/Laser_Eyes/IP_WhisperRemedySweet_Laser_ShootStop_03}
 }
 {
  id=WavePlayer_9
  properties {soundWaveAssetPtr=/Sprites_WhisperRemedySweetSprite/Assets/Sounds/Waves/VO/VO_S33_WhisperRemedySweetSprite_Emotion_Relieved_017}
 }
]
properties {
 attenuationSettings=/Sprites_WhisperRemedySweetSprite/Assets/Sounds/Settings/Laser_Eyes/IP_WhisperRemedySweet_Beam_Fire_Att
 maxDistance=18400
 soundClassObject=&Fort_WeaponRanged_Fire_3P_PSM
 volumeMultiplier=1
}
